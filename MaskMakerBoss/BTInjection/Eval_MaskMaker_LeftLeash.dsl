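; Behavior tree injection evaluator that detects when the MaskMaker boss has left its leash.
; On enter, it caches the owning Fort Athena AI bot controller and its controlled pawn.
; Starting conditions pass only if the pawn is an arena-bound MaskMaker boss and its location lies outside the controller's leash.
; Ongoing conditions always return true, so the leash is not rechecked there.
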(event EventOnEnter
  (params (OwnerComp :type Behavior%20Tree%20Component%20Object%20Reference))
  (bind _asfort_athena_aibot_controller (Utilities|Casting|CastToFortAthenaAIBotController (Components|GetOwner OwnerComp)))
  (bind _output_get (Variables|Default|SetMaskMakerController _asfort_athena_aibot_controller))
  (bind _asfort_player_pawn_athena (Utilities|Casting|CastToFortPlayerPawnAthena (Pawn|GetControlledPawn _output_get)))
  (Variables|Default|SetMaskMakerPawn _asfort_player_pawn_athena))

(fn EvaluateOngoingConditions
  (params (OwnerComp :type Behavior%20Tree%20Component%20Object%20Reference))
  (|CallParentFunction OwnerComp)
  (return true))

(fn EvaluateStartingConditions
  (params (OwnerComp :type Behavior%20Tree%20Component%20Object%20Reference))
  (|CallParentFunction OwnerComp)
  (bind _asbp_player_pawn_cherry_smoke_mask_maker_boss (Utilities|Casting|CastToBP_PlayerPawn_CherrySmoke_MaskMakerBoss (Variables|Default|GetMaskMakerPawn))
    (:then
      (if (Variables|/MaskMakerBoss/Pawn/BP_PlayerPawn_CherrySmoke_MaskMakerBoss.BP_PlayerPawn_CherrySmoke_MaskMakerBoss_C|GetIsArenaBound _asbp_player_pawn_cherry_smoke_mask_maker_boss)
        (return (not (Class|FortAthenaAIBotController|IsLocationInsideLeash (Variables|Default|GetMaskMakerController) (Transformation|GetActorLocation (Variables|Default|GetMaskMakerPawn)) false 0.0)))
        (else
          _)))
    (:CastFailed
      _))
  (return false))
