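; Behavior tree evaluator that decides when the MaskMaker boss can start its Spire Shards ability.
; On enter it caches the AI controller and controlled pawn.
; Starting requires the boss to be floating, not just floated up, free of the between-cooldown tag, with room under the total shard cap for a maximum batch, and the hotfix toggle enabled.
; Ongoing conditions always pass; both evaluations call their parent but ignore its result.
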(event EventOnEnter
  (params (OwnerComp :type Behavior%20Tree%20Component%20Object%20Reference))
  (bind _asfort_athena_aibot_controller (Utilities|Casting|CastToFortAthenaAIBotController (Components|GetOwner OwnerComp)))
  (bind _output_get (Variables|Default|SetMaskMakerController _asfort_athena_aibot_controller))
  (bind _asfort_player_pawn_athena (Utilities|Casting|CastToFortPlayerPawnAthena (Pawn|GetControlledPawn _output_get)))
  (Variables|Default|SetMaskMakerPawn _asfort_player_pawn_athena))

(fn EvaluateOngoingConditions
  (params (OwnerComp :type Behavior%20Tree%20Component%20Object%20Reference))
  (|CallParentFunction OwnerComp)
  (return true))

(fn EvaluateStartingConditions
  (params (OwnerComp :type Behavior%20Tree%20Component%20Object%20Reference))
  (bind _maskmakerpawn (Variables|Default|GetMaskMakerPawn))
  (|CallParentFunction OwnerComp)
  (bind _bool (Class|MaskMakerPawnInterface|GetIsFloating(Message) _maskmakerpawn))
  (bind _bool_2 (Class|MaskMakerPawnInterface|GetJustFloatedUp(Message) _maskmakerpawn))
  (bind _amount (Class|MaskMakerPawnInterface|GetTotalShardSpawnedAmount(Message) _maskmakerpawn))
  (return (Math|Boolean|ANDBoolean _bool (not _bool_2) (not (GameplayTags|HasMatchingGameplayTag _maskmakerpawn "(TagName=\"MaskMakerBoss.Ability.BetweenCooldown\")")) (>= (- (ScalableFloat|GetValueasInteger (Variables|Hotfix|GetMaxTotalShardAmount_Hf)) _amount) (ScalableFloat|GetValueasInteger (Variables|Hotfix|GetMaxShardAmount_Hf))) (ScalableFloat|GetValueasBool (Variables|Hotfix|GetCanUseSpireShards_Hf)))))
