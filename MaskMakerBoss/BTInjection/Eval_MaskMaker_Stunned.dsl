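; Behavior tree evaluator for the MaskMaker boss that gates behavior on the boss's stunned state, apparently for an injected branch.
; On enter, it casts the owning AI bot controller and its controlled pawn, caching both.
; The starting condition returns the stunned flag queried from that pawn through MaskMakerPawnInterface.
; The ongoing condition always returns true, so the stunned flag is not rechecked there once started.
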(event EventOnEnter
  (params (OwnerComp :type Behavior%20Tree%20Component%20Object%20Reference))
  (bind _asfort_athena_aibot_controller (Utilities|Casting|CastToFortAthenaAIBotController (Components|GetOwner OwnerComp)))
  (bind _output_get (Variables|Default|SetMaskMakerController _asfort_athena_aibot_controller))
  (bind _asfort_player_pawn_athena (Utilities|Casting|CastToFortPlayerPawnAthena (Pawn|GetControlledPawn _output_get)))
  (Variables|Default|SetMaskMakerPawn _asfort_player_pawn_athena))

(fn EvaluateOngoingConditions
  (params (OwnerComp :type Behavior%20Tree%20Component%20Object%20Reference))
  (|CallParentFunction OwnerComp)
  (return true))

(fn EvaluateStartingConditions
  (params (OwnerComp :type Behavior%20Tree%20Component%20Object%20Reference))
  (|CallParentFunction OwnerComp)
  (bind _bool (Class|MaskMakerPawnInterface|GetIsinStunnedState(Message) (Variables|Default|GetMaskMakerPawn)))
  (return _bool))
