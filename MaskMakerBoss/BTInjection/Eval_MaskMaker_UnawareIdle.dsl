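; Behavior tree evaluator that decides when the MaskMaker boss enters and stays in its unaware idle.
; On enter it caches controller and pawn; ongoing checks simply return the pawn interface's DoUnawareIdle flag.
; Starting checks require a living, unstunned boss, passing if already flagged and Unaware, or setting the flag when the target is missing or downed.
; A target outside the leash and out of sight or beam range also triggers idle after a timed delay.
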(event EventOnEnter
  (params (OwnerComp :type Behavior%20Tree%20Component%20Object%20Reference))
  (bind _asfort_athena_aibot_controller (Utilities|Casting|CastToFortAthenaAIBotController (Components|GetOwner OwnerComp)))
  (bind _output_get (Variables|Default|SetMaskMakerController _asfort_athena_aibot_controller))
  (bind _asfort_player_pawn_athena (Utilities|Casting|CastToFortPlayerPawnAthena (Pawn|GetControlledPawn _output_get)))
  (Variables|Default|SetMaskMakerPawn _asfort_player_pawn_athena))

(fn EvaluateOngoingConditions
  (params (OwnerComp :type Behavior%20Tree%20Component%20Object%20Reference))
  (|CallParentFunction OwnerComp)
  (bind _bool (Class|MaskMakerPawnInterface|GetDoUnawareIdle(Message) (Variables|Default|GetMaskMakerPawn)))
  (return _bool))

(fn EvaluateStartingConditions
  (params (OwnerComp :type Behavior%20Tree%20Component%20Object%20Reference))
  (bind _isinstunnedstatelocal (LocalVariables|GetIsInStunnedStateLocal))
  (bind _returnvalue (not _isinstunnedstatelocal))
  (bind _maskmakerpawn (Variables|Default|GetMaskMakerPawn))
  (bind _returnvalue_2 (FortPawn|IsDead _maskmakerpawn))
  (bind _returnvalue_3 (not _returnvalue_2))
  (bind _maskmakerpawn_2 (Variables|Default|GetMaskMakerPawn))
  (bind _maskmakercontroller (Variables|Default|GetMaskMakerController))
  (|CallParentFunction OwnerComp)
  (bind _bool (Class|MaskMakerPawnInterface|GetDoUnawareIdle(Message) _maskmakerpawn_2))
  (bind _bool_2 (Class|MaskMakerPawnInterface|GetIsinStunnedState(Message) _maskmakerpawn_2))
  (LocalVariables|SetIsInStunnedStateLocal _bool_2)
  (if (Math|Boolean|ANDBoolean _bool (not _bool_2) (not (FortPawn|IsDead _maskmakerpawn_2)) (Utilities|Enum|Equal(Enum) (Class|FortAthenaAIBotController|GetAlertLevel (Variables|Default|GetMaskMakerController)) "Unaware"))
    (return true)
    (else
      (bind _targetactor (Class|BPFLAICommonFunctions|GetTargetActorfromBlackboard _maskmakerpawn))
      (Utilities|IsValid _targetactor
        (:Is%20Valid
          (if (Health|IsDBNO _targetactor)
            (if (not (and _returnvalue _returnvalue_3)) :!label _label_4
              (Class|MaskMakerPawnInterface|SetDoUnawareIdle(Message) (Variables|Default|GetMaskMakerPawn) false :!label _label)
              (return false))
            (elif (not (Class|FortAthenaAIBotController|IsLocationInsideLeash _maskmakercontroller (Transformation|GetActorLocation _targetactor) false 0.0))
              (if (== (Variables|Default|GetTargetLostDelay) 0)
                (Variables|Default|SetTargetLostDelay (Fortnite|Sync|Time|GetServerWorldRealTimeSeconds)))
              (if (Math|Boolean|ANDBoolean _returnvalue _returnvalue_3 (or (not (Controller|LineOfSightTo :self _maskmakercontroller :Other _targetactor :bAlternateChecks false)) (not (< (Math|Vector|Distance(Vector) (Transformation|GetActorLocation _maskmakerpawn) (Transformation|GetActorLocation _targetactor)) (ScalableFloat|GetValueatLevel (Variables|Hotfix|GetBeamMaxRange_Hf))))))
                (if (> (Math|Float|Absolute(Float) (- (Variables|Default|GetTargetLostDelay) (Fortnite|Sync|Time|GetServerWorldRealTimeSeconds))) (Variables|Default|GetDelayTimeSinceLeashExit))
                  (Class|MaskMakerPawnInterface|SetDoUnawareIdle(Message) (Variables|Default|GetMaskMakerPawn) true :!label _label_3)
                  (return true)
                  (else
                    _))
                (else
                  (Variables|Default|SetTargetLostDelay 0.0 :!label _label_2)))
              (goto _label)
              (else
                (goto _label_2))))
          (goto _label_3))
        (:Is%20Not%20Valid
          _))
      (goto _label_4))))
